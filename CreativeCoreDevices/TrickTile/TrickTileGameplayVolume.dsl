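; Gameplay volume for the Creative Trick Tile device that manages how its zone mesh displays and cleans up its playspace when destroyed.
; Construction hides the actor in game when in the editor, and BeginPlay unhides it at runtime.
; Change Visibility shows the overlap mesh when requested or in editor, async-loading the zone material only then and otherwise assigning a grid material to save memory.
; EndPlay destroys the playspace if the actor is destroyed.
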(event EventBeginPlay
  (|CallParentFunction)
  (Rendering|SetActorHiddenInGame false) ; Clearing Actor Hidden in Game (set on construction if we're the editor).\nWe want to be visible when not in editor (which won't run this, because\nit doesn't Begin Play)
)

(event Custom|Change%20Visibility
  (params (Visible :type Boolean))
  (bind _returnvalue (Editor|InEditor))
  (bind _returnvalue_2 (or Visible _returnvalue))
  (bind _overlapmesh (Variables|Default|GetOverlapMesh))
  ;[1 0.5771 0 0.25] Only show the Zone Material when visible, to lower memory costs (during game this will never be visible).
  (Rendering|SetVisibility (Variables|Default|GetOverlapMesh) _returnvalue_2 false)
  (if _returnvalue_2
    (bind _loaded_object (Creative|Loading|AsyncLoadEditorSafe (Variables|Default|GetZoneMaterial))
      (:Loaded
        (bind _asmaterial_interface (Utilities|Casting|CastToMaterialInterface _loaded_object))
        (Rendering|Material|SetMaterial _overlapmesh 0 _asmaterial_interface)))
    (else
      (Rendering|Material|SetMaterial _overlapmesh 0 "/Game/FN_PublicTurntable/TurntableAssets/WorldGridMaterial.WorldGridMaterial"))))

(event EventEndPlay
  (params (EndPlayReason :type EEndPlayReason%20Enum :default "Destroyed"))
  (if (Utilities|Enum|Equal(Enum) EndPlayReason "Destroyed")
    (Class|PlayspaceLibrary|DestroyPlayspace (CallFunction|GetPlayspace))))

(fn ConstructionScript
  (Rendering|SetActorHiddenInGame (Editor|InEditor)) ; Actor is always hidden in game in editor\n(since that's 'Game Mode')
)
